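; Calendar-driven progressive POI prop that swaps its default building foundation for the dynamic foundations of the current progress stage.
; On authority, it takes the stage from the last digit of the first active calendar event outside Creative, or from a debug override.
; It walks stage thresholds in reverse, queuing each reached foundation once, so the first entry found sets any Z offset.
; Queued transforms are applied after the loop, alongside enabling the foundations.
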(event Custom|DisableDefaultFoundation
  ;[1 1 1] Disable default foundation
  (Class|BuildingFoundation|SetDynamicFoundationEnabled (Variables|Default|GetDefaultFoundation) false))

(event Custom|LevelProgress
  (params (Progress :type Integer))
  ;[1 1 1] Enable Foundation if past the "threshold"
  (Variables|Default|SetProgressStage Progress)
  (CallFunction|DisableDefaultFoundation)
  (bind _keys (Utilities|Map|Keys (Variables|Default|GetNewLevels)))
  (bind (_arrayindex _arrayelement) (Utilities|Array|ReverseforEachLoop _keys)
    (:LoopBody
      (Development|PrintString (Utilities|String|ToString(Integer) _arrayelement))
      (if (>= (Variables|Default|GetProgressStage) _arrayelement)
        (bind _value (Utilities|Map|Find (Variables|Default|GetNewLevels) _arrayelement))
        (bind _value_levelsetup_28_5b07da4e4ffa7913e1c2ef864c0bdc68 (Utilities|Struct|BreakPOICalendarProgData _value))
        (for _array_element _value_levelsetup_28_5b07da4e4ffa7913e1c2ef864c0bdc68
          (bind (_array_element_buildingfoundation_30_5b07da4e4ffa7913e1c2ef864c0bdc68 _array_element_zoffset_31_c1b130794672fc8b513c78bbbeb310d6) (Utilities|Struct|BreakPOICalendarEntryData _array_element))
          (CallFunction|LoadLevel _array_element_buildingfoundation_30_5b07da4e4ffa7913e1c2ef864c0bdc68 _array_element_zoffset_31_c1b130794672fc8b513c78bbbeb310d6))))
    (:Completed
      (Utilities|FlowControl|Sequence
        (:then_0
          (bind _keys_2 (Utilities|Map|Keys (Variables|Default|GetLevelTransforms)))
          (for _array_element_2 _keys_2
            (Class|BuildingFoundation|SetDynamicFoundationTransform _array_element_2 (Utilities|Map|Find (Variables|Default|GetLevelTransforms) _array_element_2))))
        (:then_1
          (for _array_element_3 (Variables|Default|GetSetEnabledDeferred)
            (Class|BuildingFoundation|SetDynamicFoundationEnabled _array_element_3 true)))))))

(event Custom|LoadLevel
  (params (Foundation :type Building%20Foundation%20Object%20Reference) (ZOffset :type Float%20%28double-precision%29))
  (if (not (Utilities|Array|ContainsItem (Variables|Default|GetSetEnabledDeferred) Foundation))
    (Utilities|Array|AddUnique (Variables|Default|GetSetEnabledDeferred) Foundation)
    (if (not (== ZOffset -1.0))
      (Utilities|Map|Add (Variables|Default|GetLevelTransforms) Foundation (Math|Transform|ComposeTransforms (Math|Transform|MakeTransform (Math|Vector|MakeVector 0.0 0.0 ZOffset) "0, 0, 0" "1.000000,1.000000,1.000000") (Transformation|GetActorTransform Foundation))))))

(event EventBeginPlay
  (bind _debugprogressstep (Variables|Default|GetDebugProgressStep))
  (Networking|SwitchHasAuthority
    (:Authority
      (if (== _debugprogressstep -1)
        (bind _gamestate (Class|FortAsyncActionCurrentPlaylistReady|PlaylistDataReadyAsync)
          (:OnReady
            (if (not (GameplayTags|HasTag (Class|FortGameStateAthena|GetAthenaPlaylistContextTags _gamestate) "(TagName=\"Athena.Creative\")"))
              (bind _array_element (Utilities|Array|ForEachLoopwithBreak (Variables|Default|GetEvents))
                (:LoopBody
                  (bind _biseventactive (CalendarEvents|GetCalendarEventInformation _array_element))
                  (if _biseventactive
                    (CallFunction|BreakLoop)
                    (CallFunction|LevelProgress (Utilities|String|StringToInteger (Utilities|String|Right _array_element 1)))))))))
        (else
          (CallFunction|LevelProgress _debugprogressstep))))))

(event Custom|BreakLoop
  (bind _array_element (Utilities|Array|ForEachLoopwithBreak (Variables|Default|GetEvents))
    (:LoopBody
      (bind _biseventactive (CalendarEvents|GetCalendarEventInformation _array_element))
      (if _biseventactive
        (CallFunction|BreakLoop)
        (CallFunction|LevelProgress (Utilities|String|StringToInteger (Utilities|String|Right _array_element 1)))))))

(fn ConstructionScript)
