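; Calendar-driven POI selector that, on the server, swaps a default BuildingFoundation for event-specific dynamic foundations mapped by calendar event name.
; Once the mesh network is ready, it can load every mapped foundation or watch calendar events and load them as they activate.
; Once playlist data is ready in non-Creative playlists, it enables the first active event's foundation, otherwise the default.
; A debug foundation name can override this selection.
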(event EventBeginPlay
  (bind _debugfoundation (Variables|Default|GetDebugFoundation))
  (Networking|SwitchHasAuthority
    (:Authority
      (if (Utilities|String|EqualExactly(String) _debugfoundation)
        (Utilities|FlowControl|Sequence
          (:then_0
            (bind _nodetype (Class|FortAsyncActionMeshNetworkReady|MeshNetworkReadyAsync)
              (:OnReady
                (switch Utilities|FlowControl|Switch|SwitchonEMeshNetworkNodeType _nodetype
                  (:Root
                    _)
                  (:Inner
                    _))
                (bind _keys (Utilities|Map|Keys (Variables|Default|GetNew)))
                (if (Variables|Default|GetAlwaysLoadOnMesh)
                  (for _array_element _keys
                    (CallFunction|TryLoadMap _array_element))
                  (else
                    (bind _eventname (MCPContext|WatchCalendarEvents _keys)
                      (:OnEventActive
                        (CallFunction|TryLoadMap _eventname))))))))
          (:then_1
            (bind _gamestate (Class|FortAsyncActionCurrentPlaylistReady|PlaylistDataReadyAsync)
              (:OnReady
                (if (not (GameplayTags|HasTag (Class|FortGameStateAthena|GetAthenaPlaylistContextTags _gamestate) "(TagName=\"Athena.Creative\")"))
                  (bind _keys_2 (Utilities|Map|Keys (Variables|Default|GetNew)))
                  (bind _array_element_2 (Utilities|Array|ForEachLoopwithBreak _keys_2)
                    (:LoopBody
                      (if (Variables|Default|GetLevelSetUp)
                        (CallFunction|Break)
                        (else
                          (bind _biseventactive (CalendarEvents|GetCalendarEventInformation _array_element_2))
                          (if _biseventactive
                            (CallFunction|TryLoadMap _array_element_2)))))
                    (:Completed
                      (if (not (Variables|Default|GetLevelSetUp))
                        (Class|BuildingFoundation|SetDynamicFoundationEnabled (Variables|Default|GetDefaultFoundation) true)))))))))
        (else
          (Class|BuildingFoundation|SetDynamicFoundationEnabled (Utilities|Map|Find (Variables|Default|GetNew) _debugfoundation) true)
          (Class|BuildingFoundation|SetDynamicFoundationEnabled (Variables|Default|GetDefaultFoundation) false))))))

(event Custom|Break
  (bind _array_element (Utilities|Array|ForEachLoopwithBreak (stale EventBeginPlay _keys_2))
    (:LoopBody
      (if (Variables|Default|GetLevelSetUp)
        (CallFunction|Break)
        (else
          (bind _biseventactive (CalendarEvents|GetCalendarEventInformation _array_element))
          (if _biseventactive
            (CallFunction|TryLoadMap _array_element)))))
    (:Completed
      (if (not (Variables|Default|GetLevelSetUp))
        (Class|BuildingFoundation|SetDynamicFoundationEnabled (Variables|Default|GetDefaultFoundation) true)))))

(event Custom|RuntimeFoundationSwap
  (bind _keys (Utilities|Map|Keys (Variables|Default|GetNew)))
  (for _array_element _keys
    (CallFunction|TryLoadMap _array_element)))

(event Custom|TryLoadMap
  (params (Key :type String%20%28by%20ref%29))
  (bind _new (Variables|Default|GetNew))
  (bind (_value _returnvalue) (Utilities|Map|Find _new Key))
  ;[1 1 1] TryLoadMap
  (if _returnvalue
    (Class|BuildingFoundation|SetDynamicFoundationEnabled _value true)
    (Class|BuildingFoundation|SetDynamicFoundationEnabled (Variables|Default|GetDefaultFoundation) false)
    (Variables|Default|SetLevelSetUp true)))

(fn ConstructionScript)
